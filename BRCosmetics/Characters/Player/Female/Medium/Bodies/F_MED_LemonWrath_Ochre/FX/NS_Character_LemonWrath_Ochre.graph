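# Niagara cosmetic FX for the LemonWrath Ochre female character body: small spinning cubes and soft glow sprites attached to the skeleton.
# Three IdleCubes emitters spawn drifting mesh-plus-sprite particles at the shoulders and upper spine, while MainGlow holds six looping glows on arm and right-leg bones.
# User.VariantColor indexes arrays that tint or hide layers, with cubes visible only in variant 0.
# Front-end, platform and distance modifiers adjust opacity or camera offset.
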NiagaraSystem /BRCosmetics/Characters/Player/Female/Medium/Bodies/F_MED_LemonWrath_Ochre/FX/NS_Character_LemonWrath_Ochre
dependencies {
 usedDataInterfaces [
  &NiagaraDataInterfaceArrayFloat2
  &NiagaraDataInterfaceArrayFloat
  &NiagaraDataInterfacePlatformSet
  &NiagaraDataInterfaceArrayColor
  &NiagaraDataInterfaceSkeletalMesh
  &NiagaraDataInterfaceArrayFloat3
  &NiagaraDataInterfaceVectorCurve
  &NiagaraDataInterfaceCurve
  &NiagaraDataInterfaceArrayInt32
 ]
 usedDynamicInputs []
 usedModules [
  &self:SystemSpawnScript.SystemScriptSource.SystemScriptGraph.NiagaraNodeAssignment_0.SetVariables_#1
  &SystemState
 ]
}
emitters {
 IdleCubes_Pit001 {
  emitterUpdateScript [
   &EmitterState(Life Cycle Mode=::NewEnumerator1, Loop Duration=1, UseLoopDelay=false)
   &SpawnRate(SpawnRate=4, Use Spawn Probability=false)
  ]
  particleSpawnScript [
   &InitializeParticle(
    Color=&MultiplyLinearColor(
     Linear Color=&RandomRangeLinearColor(Minimum=<rgba 0.2 0.2 0.2 1>, Maximum=<rgba 0.8 0.8 0.8 1>)
     Scale Factor=&MakeVector4_FromLinearColor(LinearColor=$Particles.Color)
    )
    Lifetime=&RandomRangeFloat(Minimum=2, Maximum=1)
    Mesh Scale Mode=::NewEnumerator1
    Mesh Uniform Scale Max=0.1
    Mesh Uniform Scale Min=0.3
    UsePositionOffset=false
    Write Material Random=true
   )
   &InitialMeshOrientation(Use Rotation Vector=true)
   &CameraOffset(Camera Offset Amount=2)
  ]
  particleUpdateScript [
   &ParticleState()
   &UpdateMeshOrientation(
    Rotation Rate=0.5
    Rotation Vector=&VectorFromFloat(Value=&RandomRangeFloat(Minimum=-1, Maximum=1))
   )
   &Color(Color=&SelectLinearColorFromArray(
     Array Sampling Mode=::NewEnumerator1
     Direct Array Index=&MakeIntFromFloat(Float=$User.VariantColor)
     Color Selection Array=DataInterface() {propertyValues {ColorData [<rgba 0.651 0.1628 0 2>, <rgba 0 0 0 0>, <rgba 0 0 0 0>, <rgba 0 0 0 0>]}}
     Random Seed=?
    ))
   &SkeletalMeshLocation(
    Bone / Socket Index=&SelectIntFromArray(Direct Array Index=$Particles.PartitionParticles.Partition, Int Selection Array=DataInterface() {propertyValues {IntData [2]}})
    Bone Sampling Mode=::NewEnumerator4
    MaskSpawnGroup=false
    Orientation Sampling=::NewEnumerator1
    Sampled Position Offset=<0 13 0>
    Skeletal Mesh=DataInterface() {propertyValues {
      DefaultMesh=null
      ExcludeBoneName=root
      OptionalComponentClass=null
      PreviewMesh=/BRCosmetics/Characters/Player/Female/Medium/Bodies/F_MED_LemonWrath_Ochre/Meshes/F_MED_LemonWrath_Ochre
      SkinningMode=SkinOnTheFly
      SoftSourceActor=null
      SourceMode=Default
      UvSetIndex=0
      WholeMeshLOD=-1
      bExcludeBone=false
      bReadDeformedGeometry=true
      bRequireCurrentFrameData=true
      ComponentTags []
      FilteredBones [upperarm_l,upperarm_r,spine_04]
      FilteredSockets []
      MeshUserParameter {parameter {name=None, typeDefHandle {registeredTypeIndex=1}, varData []}}
      SamplingRegions []
     }}
   ) {module=SkeletalMeshLocation001}
   &ScaleMeshSize(Scale Factor=&VectorFromCurve(
     VectorCurve=DataInterface() {propertyValues {
       XCurve {keys [<at 0 1>,<at 0.0821 0.9489>,<at 0.1617 1.1754>,<at 0.8653 1.1811>,<at 0.9758 1.3967>]}
       YCurve {keys [<at 0 1>,<at 0.1192 1.3385>,<at 0.1592 1.1826>,<at 0.9654 1.1803>]}
       ZCurve {keys [<at 0 0.577>,<at 0.0503 1.3231>,<at 0.1202 1.0349>,<at 0.1592 1.1826>,<at 0.9862 1.2393>]}
      }}
     Scale Curve=&Multiply_VectorByFloat(
      Vector=&RandomRangeVector(Minimum=<0.75 0.75 0.75>, Maximum=<1.25 1.25 1.25>)
      Float=&FloatFromCurve(
       FloatCurve=DataInterface() {propertyValues {Curve {keys [
           {arriveTangentWeight=0.0895, interpMode=RCIM_Cubic, leaveTangentWeight=0.2109, tangentMode=RCTM_User, time=0.2, value=1}
           {interpMode=RCIM_Cubic, tangentMode=RCTM_User, time=0.8499, value=1}
           {arriveTangent=-4.1587, interpMode=RCIM_Cubic, leaveTangent=-4.1587, tangentMode=RCTM_User, time=1, value=-0.0118}
          ]}}}
       Scale Curve=&SelectFloatFromArray(
        Direct Array Index=&MakeIntFromFloat(Float=$User.VariantColor)
        Float Selection Array=DataInterface() {propertyValues {FloatData [1,0,0,0]}}
       )
      )
     )
    ))
   &DynamicMaterialParameters(
    Index 1 Param 1=5
    Index 1 Param 2=5
    Module.Index0Float0Write=true
    Module.Index0Float1Write=true
    Module.Index0Float2Write=true
    Module.Index0Float3Write=true
    Module.Index1Float0Write=true
    Module.Index1Float1Write=true
    Module.Index1Float2Write=true
    Module.Index1Float3Write=true
    Write Parameter Index 1=::NewEnumerator1
   )
   /SetVariables(Particles.Position=&AddVectorToPosition(
     Position=$Particles.Position
     Vector=&VectorFromCurve(
      VectorCurve=DataInterface() {propertyValues {
        XCurve {keys [<at 0 0.8>, {arriveTangent=2, time=0.25, value=1}, <at 0.5 0.8>, <at 0.75 1>, <at 1 0.8>]}
        YCurve {defaultValue=0, keys [<at 0 0.8>,<at 0.25 1>,<at 0.5 0.8>,<at 0.75 1>,<at 1 0.8>]}
        ZCurve {keys [<at 0 0>,<at 1 1>]}
       }}
      Scale Curve=&Multiply_VectorByFloat(Vector=&RandomRangeVector(Minimum=<-1 -1 2.5>, Maximum=<1 1 1>), Float=2.5)
     )
    )) {isSetParameters=true}
   &FN_Sprite_Attributes(
    Color=&MakeLinearColorFromVectorAndFloat(
     "Vector (RGB)"=&MakeVectorFromLinearColorRGB(Color=<rgba 1 0.25 0 1>)
     "Float (Alpha)"=&FloatFromCurve(
      FloatCurve=DataInterface() {propertyValues {Curve {defaultValue=1, keys [<at 0 0>, <at 0.2 1>, <at 0.8 1>, {tangentMode=RCTM_User, time=1, value=0}]}}}
      Scale Curve=&Multiply_Float(
       A=&SelectFloatFromArray(
        Direct Array Index=&MakeIntFromFloat(Float=$User.VariantColor)
        Float Selection Array=DataInterface() {propertyValues {FloatData [1,0,0,0]}}
       )
       B=&FN_FrontEndFloatModifier(Gameplay=0.3, Front End=0.2)
      )
     )
    )
    Create Color Attribute=true
    Create SpriteSize Attribute=true
    Module Output:ENiagara_AttributeSamplingApplyOutput=::NewEnumerator1
    SpriteSize=&Vector2DFromFloat(Value=&RandomRangeFloat(Minimum=8, Maximum=10))
   )
   &ScaleColor(
    Color Value To Scale=$Particles.Color
    Scale Alpha=&FloatFromCurve(
     FloatCurve=DataInterface() {propertyValues {Curve {defaultValue=1, keys [<at 0.2 1>, <at 0.8 1>, {tangentMode=RCTM_User, time=1, value=0}]}}}
     Scale Curve=&FN_FrontEndFloatModifier(Gameplay=1, Front End=0.8)
    )
    ScaleA=true
    ScaleRGB=true
   )
  ]
  rendererClasses [&NiagaraMeshRendererProperties, &NiagaraSpriteRendererProperties]
  renderers [
   {rendererClass=&NiagaraMeshRendererProperties, rendererIndex=0}
   {rendererClass=&NiagaraSpriteRendererProperties, rendererIndex=1}
  ]
 }
 IdleCubes_shoulder {
  emitterUpdateScript [
   &EmitterState(Life Cycle Mode=::NewEnumerator1, Loop Duration=1, UseLoopDelay=false)
   &SpawnBurst_Instantaneous(Spawn Count=4, Use Spawn Probability=false, UseLoopCountLimit=false)
   &SpawnRate(SpawnRate=10, Use Spawn Probability=false)
  ]
  particleSpawnScript [
   &PartitionParticles()
   &InitializeParticle(
    Color=&MultiplyLinearColor(
     Linear Color=&RandomRangeLinearColor(Minimum=<rgba 0.2 0.2 0.2 1>, Maximum=<rgba 0.8 0.8 0.8 1>)
     Scale Factor=&MakeVector4_FromLinearColor(LinearColor=$Particles.Color)
    )
    Lifetime=&RandomRangeFloat(Minimum=1, Maximum=2)
    Mesh Scale Mode=::NewEnumerator1
    Mesh Uniform Scale Max=0.1
    Mesh Uniform Scale Min=0.3
    UsePositionOffset=false
    Write Material Random=true
   )
   &InitialMeshOrientation(Use Rotation Vector=true)
   &CameraOffset(Camera Offset Amount=1)
   &ShapeLocation(
    Box Size=<7 8 5>
    MaskSpawnGroup=false
    Override Seed=false
    Shape Origin=&VectorFromFloat(Value=0)
    Shape Primitive=::NewEnumerator2
    Surface Only=false
    UseOwnerScale=true
   )
  ]
  particleUpdateScript [
   &ParticleState()
   &SkeletalMeshLocation(
    Bone / Socket Index=$Particles.PartitionParticles.Partition
    Bone Sampling Mode=::NewEnumerator4
    MaskSpawnGroup=false
    Orientation Sampling=::NewEnumerator1
    Sampled Position Offset=&AddVectorToPosition(
     Position=$Particles.ShapeLocation.ShapePosition
     Vector=&Lerp_Vector(
      A=&SelectVectorFromArray(
       Direct Array Index=$Particles.PartitionParticles.Partition
       Vector Selection Array=DataInterface() {propertyValues {FloatData [<8 -2 1>, <-8 2 -1>]}}
      )
      B=&SelectVectorFromArray(
       Direct Array Index=$Particles.PartitionParticles.Partition
       Vector Selection Array=DataInterface() {propertyValues {FloatData [<8 -2 7>, <-8 2 -7>]}}
      )
      Alpha=$Particles.NormalizedAge
     )
    )
    Skeletal Mesh=$System.Skeletal Mesh Reader
   )
   &UpdateMeshOrientation(
    Rotation Rate=0.5
    Rotation Vector=&VectorFromFloat(Value=&RandomRangeFloat(Minimum=-1, Maximum=1))
   )
   &Color(Color=&SelectLinearColorFromArray(
     Array Sampling Mode=::NewEnumerator1
     Direct Array Index=&MakeIntFromFloat(Float=$User.VariantColor)
     Color Selection Array=DataInterface() {propertyValues {ColorData [<rgba 0.651 0.1628 0 2>, <rgba 0 0 0 0>, <rgba 0 0 0 0>, <rgba 0 0 0 0>]}}
     Random Seed=?
    ))
   &DynamicMaterialParameters(
    Index 1 Param 1=5
    Index 1 Param 2=5
    Module.Index0Float0Write=true
    Module.Index0Float1Write=true
    Module.Index0Float2Write=true
    Module.Index0Float3Write=true
    Module.Index1Float0Write=true
    Module.Index1Float1Write=true
    Module.Index1Float2Write=true
    Module.Index1Float3Write=true
    Write Parameter Index 1=::NewEnumerator1
   )
   &ScaleMeshSize(Scale Factor=&VectorFromCurve(
     VectorCurve=DataInterface() {propertyValues {
       XCurve {keys [<at 0 1>,<at 0.0821 0.9489>,<at 0.1617 1.1754>,<at 0.8653 1.1811>,<at 0.9758 1.3967>]}
       YCurve {keys [<at 0 1>,<at 0.1192 1.3385>,<at 0.1592 1.1826>,<at 0.9654 1.1803>]}
       ZCurve {keys [<at 0 0.577>,<at 0.0503 1.3231>,<at 0.1202 1.0349>,<at 0.1592 1.1826>,<at 0.9862 1.2393>]}
      }}
     Scale Curve=&Multiply_VectorByFloat(
      Vector=&RandomRangeVector(Minimum=<0.75 0.75 0.75>, Maximum=<1.25 1.25 1.25>)
      Float=&FloatFromCurve(
       FloatCurve=DataInterface() {propertyValues {Curve {keys [
           {arriveTangentWeight=0.0895, interpMode=RCIM_Cubic, leaveTangentWeight=0.2109, tangentMode=RCTM_User, time=0.1037, value=0}
           {arriveTangent=1.3401, interpMode=RCIM_Cubic, leaveTangent=1.3401, time=0.25, value=1.003}
           {interpMode=RCIM_Cubic, tangentMode=RCTM_User, time=0.8499, value=1}
           {arriveTangent=-4.1587, interpMode=RCIM_Cubic, leaveTangent=-4.1587, tangentMode=RCTM_User, time=1, value=-0.0118}
          ]}}}
       Scale Curve=&SelectFloatFromArray(
        Direct Array Index=&MakeIntFromFloat(Float=$User.VariantColor)
        Float Selection Array=DataInterface() {propertyValues {FloatData [1,0,0,0]}}
       )
      )
     )
    ))
   &FN_Sprite_Attributes(
    Color=&MakeLinearColorFromVectorAndFloat(
     "Vector (RGB)"=&MakeVectorFromLinearColorRGB(Color=<rgba 1 0.25 0 1>)
     "Float (Alpha)"=&FloatFromCurve(
      FloatCurve=DataInterface() {propertyValues {Curve {defaultValue=1, keys [<at 0 0>, <at 0.2 1>, <at 0.8 1>, {tangentMode=RCTM_User, time=1, value=0}]}}}
      Scale Curve=&Multiply_Float(
       A=&SelectFloatFromArray(
        Direct Array Index=&MakeIntFromFloat(Float=$User.VariantColor)
        Float Selection Array=DataInterface() {propertyValues {FloatData [1,0,0,0]}}
       )
       B=&FN_FrontEndFloatModifier(Gameplay=0.3, Front End=0.2)
      )
     )
    )
    Create Color Attribute=true
    Create SpriteSize Attribute=true
    Module Output:ENiagara_AttributeSamplingApplyOutput=::NewEnumerator1
    SpriteSize=&Vector2DFromFloat(Value=&RandomRangeFloat(Minimum=8, Maximum=10))
   )
   &ScaleColor(
    Color Value To Scale=$Particles.Color
    Scale Alpha=&FloatFromCurve(
     FloatCurve=DataInterface() {propertyValues {Curve {defaultValue=1, keys [<at 0.2 1>, <at 0.8 1>, {tangentMode=RCTM_User, time=1, value=0}]}}}
     Scale Curve=&Multiply_Float(
      A=&FN_FrontEndFloatModifier(Gameplay=1, Front End=0.8)
      B=&FN_LOD_Distance_Remap(Start Distance=100, End Distance=200, Minimum Output=0.1, Maximum Output=1, Clamp Output=true)
     )
    )
    ScaleA=true
    ScaleRGB=true
   )
  ]
  rendererClasses [&NiagaraMeshRendererProperties, &NiagaraSpriteRendererProperties]
  renderers [
   {rendererClass=&NiagaraMeshRendererProperties, rendererIndex=0}
   {rendererClass=&NiagaraSpriteRendererProperties, rendererIndex=1}
  ]
 }
 IdleCubes_shoulder001 {
  emitterUpdateScript [
   &EmitterState(Life Cycle Mode=::NewEnumerator1, Loop Duration=1, UseLoopDelay=false)
   &SpawnBurst_Instantaneous(Spawn Count=4, Use Spawn Probability=false, UseLoopCountLimit=false)
   &SpawnRate(SpawnRate=16, Use Spawn Probability=false)
  ]
  particleSpawnScript [
   &PartitionParticles()
   &InitializeParticle(
    Color=&MultiplyLinearColor(
     Linear Color=&RandomRangeLinearColor(Minimum=<rgba 0.2 0.2 0.2 1>, Maximum=<rgba 0.8 0.8 0.8 1>)
     Scale Factor=&MakeVector4_FromLinearColor(LinearColor=$Particles.Color)
    )
    Lifetime=&RandomRangeFloat(Minimum=1, Maximum=2)
    Mesh Scale Mode=::NewEnumerator1
    Mesh Uniform Scale Max=0.1
    Mesh Uniform Scale Min=0.25
    UsePositionOffset=false
    Write Material Random=true
   )
   &InitialMeshOrientation(Use Rotation Vector=true)
   &CameraOffset(Camera Offset Amount=1)
   &ShapeLocation(
    Box Size=<4 17 5>
    MaskSpawnGroup=false
    Override Seed=false
    Shape Origin=&VectorFromFloat(Value=0)
    Shape Primitive=::NewEnumerator2
    Surface Only=false
    UseOwnerScale=true
   )
  ]
  particleUpdateScript [
   &ParticleState()
   &SkeletalMeshLocation(
    Bone / Socket Index=$Particles.PartitionParticles.Partition
    Bone Sampling Mode=::NewEnumerator4
    MaskSpawnGroup=false
    Orientation Sampling=::NewEnumerator1
    Sampled Position Offset=&AddVectorToPosition(
     Position=$Particles.ShapeLocation.ShapePosition
     Vector=&Lerp_Vector(
      A=&SelectVectorFromArray(
       Direct Array Index=$Particles.PartitionParticles.Partition
       Vector Selection Array=DataInterface() {propertyValues {FloatData [<7 -1 -1>, <-7 1 1>]}}
      )
      B=&SelectVectorFromArray(
       Direct Array Index=$Particles.PartitionParticles.Partition
       Vector Selection Array=DataInterface() {propertyValues {FloatData [<7 -1 5>, <-7 1 -5>]}}
      )
      Alpha=$Particles.NormalizedAge
     )
    )
    Skeletal Mesh=$System.Skeletal Mesh Reader
   )
   &UpdateMeshOrientation(
    Rotation Rate=0.5
    Rotation Vector=&VectorFromFloat(Value=&RandomRangeFloat(Minimum=-1, Maximum=1))
   )
   &Color(Color=&SelectLinearColorFromArray(
     Array Sampling Mode=::NewEnumerator1
     Direct Array Index=&MakeIntFromFloat(Float=$User.VariantColor)
     Color Selection Array=DataInterface() {propertyValues {ColorData [<rgba 0.651 0.1628 0 2>, <rgba 0 0 0 0>, <rgba 0 0 0 0>, <rgba 0 0 0 0>]}}
     Random Seed=?
    ))
   &DynamicMaterialParameters(
    Index 1 Param 1=5
    Index 1 Param 2=5
    Module.Index0Float0Write=true
    Module.Index0Float1Write=true
    Module.Index0Float2Write=true
    Module.Index0Float3Write=true
    Module.Index1Float0Write=true
    Module.Index1Float1Write=true
    Module.Index1Float2Write=true
    Module.Index1Float3Write=true
    Write Parameter Index 1=::NewEnumerator1
   )
   &ScaleMeshSize(Scale Factor=&VectorFromCurve(
     VectorCurve=DataInterface() {propertyValues {
       XCurve {keys [<at 0 1>,<at 0.0821 0.9489>,<at 0.1617 1.1754>,<at 0.8653 1.1811>,<at 0.9758 1.3967>]}
       YCurve {keys [<at 0 1>,<at 0.1192 1.3385>,<at 0.1592 1.1826>,<at 0.9654 1.1803>]}
       ZCurve {keys [<at 0 0.577>,<at 0.0503 1.3231>,<at 0.1202 1.0349>,<at 0.1592 1.1826>,<at 0.9862 1.2393>]}
      }}
     Scale Curve=&Multiply_VectorByFloat(
      Vector=&RandomRangeVector(Minimum=<0.75 0.75 0.75>, Maximum=<1.25 1.25 1.25>)
      Float=&FloatFromCurve(
       FloatCurve=DataInterface() {propertyValues {Curve {keys [
           {arriveTangentWeight=0.0895, interpMode=RCIM_Cubic, leaveTangentWeight=0.2109, tangentMode=RCTM_User, time=0.1037, value=0}
           {arriveTangent=1.3401, interpMode=RCIM_Cubic, leaveTangent=1.3401, time=0.25, value=1.0268}
           {interpMode=RCIM_Cubic, tangentMode=RCTM_User, time=0.8499, value=1}
           {arriveTangent=-4.1587, interpMode=RCIM_Cubic, leaveTangent=-4.1587, tangentMode=RCTM_User, time=1, value=-0.0118}
          ]}}}
       Scale Curve=&SelectFloatFromArray(
        Direct Array Index=&MakeIntFromFloat(Float=$User.VariantColor)
        Float Selection Array=DataInterface() {propertyValues {FloatData [1,0,0,0]}}
       )
      )
     )
    ))
   &FN_Sprite_Attributes(
    Color=&MakeLinearColorFromVectorAndFloat(
     "Vector (RGB)"=&MakeVectorFromLinearColorRGB(Color=<rgba 1 0.25 0 1>)
     "Float (Alpha)"=&FloatFromCurve(
      FloatCurve=DataInterface() {propertyValues {Curve {defaultValue=1, keys [<at 0 0>, <at 0.2 1>, <at 0.8 1>, {tangentMode=RCTM_User, time=1, value=0}]}}}
      Scale Curve=&Multiply_Float(
       A=&SelectFloatFromArray(
        Direct Array Index=&MakeIntFromFloat(Float=$User.VariantColor)
        Float Selection Array=DataInterface() {propertyValues {FloatData [1,0,0,0]}}
       )
       B=&FN_FrontEndFloatModifier(Gameplay=0.3, Front End=0.2)
      )
     )
    )
    Create Color Attribute=true
    Create SpriteSize Attribute=true
    Module Output:ENiagara_AttributeSamplingApplyOutput=::NewEnumerator1
    SpriteSize=&Vector2DFromFloat(Value=&RandomRangeFloat(Minimum=8, Maximum=10))
   )
   &ScaleColor(
    Color Value To Scale=$Particles.Color
    Scale Alpha=&FloatFromCurve(
     FloatCurve=DataInterface() {propertyValues {Curve {defaultValue=1, keys [<at 0.2 1>, <at 0.8 1>, {tangentMode=RCTM_User, time=1, value=0}]}}}
     Scale Curve=&Multiply_Float(
      A=&FN_FrontEndFloatModifier(Gameplay=1, Front End=0.8)
      B=&FN_LOD_Distance_Remap(Start Distance=100, End Distance=200, Minimum Output=0.1, Maximum Output=1, Clamp Output=true)
     )
    )
    ScaleA=true
    ScaleRGB=true
   )
  ]
  rendererClasses [&NiagaraMeshRendererProperties, &NiagaraSpriteRendererProperties]
  renderers [
   {rendererClass=&NiagaraMeshRendererProperties, rendererIndex=0}
   {rendererClass=&NiagaraSpriteRendererProperties, rendererIndex=1}
  ]
 }
 MainGlow {
  emitterUpdateScript [
   &EmitterState(
    Inactive Response=::NewEnumerator1
    Life Cycle Mode=::NewEnumerator1
    Loop Behavior=::NewEnumerator1
    Loop Duration Mode=::NewEnumerator1
    UseLoopDelay=false
   )
   &SpawnBurst_Instantaneous(Spawn Count=6, Use Spawn Probability=false, UseLoopCountLimit=false)
  ]
  particleSpawnScript [
   &PartitionParticles(Number Of Partitions=6)
   &InitializeParticle(
    Color Mode=::NewEnumerator0
    Sprite Size=&SelectVector2DFromArray(
     Direct Array Index=$Particles.PartitionParticles.Partition
     Vector 2D Selection Array=DataInterface() {propertyValues {FloatData [<25 30>, <25 30>, <25 40>, <25 35>, <25 25>, <25 25>]}}
    )
    Sprite Size Mode=::NewEnumerator4
    UsePositionOffset=false
    Write Material Random=true
   )
   &DynamicMaterialParameters(
    Index 0 Param 1=5
    Index 0 Param 2=10
    Module.Index0Float0Write=true
    Module.Index0Float1Write=true
    Module.Index0Float2Write=true
    Module.Index0Float3Write=true
   )
   &CameraOffset(Camera Offset Amount=&FN_PickFloatDependOnPlatform(
     TargetPlatform=DataInterface() {propertyValues {Platforms {qualityLevelMask=-2}}}
     TargetPlatformActive=&SelectFloatFromArray(
      Direct Array Index=$Particles.PartitionParticles.Partition
      Float Selection Array=DataInterface() {propertyValues {FloatData [10,10,6,6,6,6]}}
     )
     TargetPlatformNoActive=2
    ))
  ]
  particleUpdateScript [
   &ParticleState(Kill Particles When Lifetime Has Elapsed=false, Loop Particles Lifetime=true)
   &Color(Color=&SelectLinearColorFromArray(
     Array Sampling Mode=::NewEnumerator1
     Direct Array Index=&MakeIntFromFloat(Float=$User.VariantColor)
     Color Selection Array=DataInterface() {propertyValues {ColorData [<rgba 1 0.3333 0 1>, <rgba 0 0 0 0>, <rgba 0 0.25 1 1>, <rgba 0 0 0 0>]}}
     Random Seed=?
    ))
   &ScaleColor(
    Color Value To Scale=$Particles.Color
    Scale Alpha=&Multiply_Float(
     A=&FN_PickFloatDependOnPlatform(
      TargetPlatform=DataInterface() {propertyValues {Platforms {qualityLevelMask=-2}}}
      TargetPlatformActive=&FN_FrontEndFloatModifier(
       Gameplay=&SelectFloatFromArray(
        Direct Array Index=$Particles.PartitionParticles.Partition
        Float Selection Array=DataInterface() {propertyValues {FloatData [1,1,0.5,0.5,2.2,2.2]}}
       )
       Front End=&SelectFloatFromArray(
        Direct Array Index=$Particles.PartitionParticles.Partition
        Float Selection Array=DataInterface() {propertyValues {FloatData [0.5,0.5,0.5,0.5,1,1]}}
       )
      )
      TargetPlatformNoActive=&FN_FrontEndFloatModifier(Gameplay=0.2, Front End=0.1)
     )
     B=&MakeCustomFloatFromBool(
      Bool=&SetBoolByFloatComparison(A=$User.VariantColor, B=2, Comparison Type=::NewEnumerator4)
      True Float=&SelectFloatFromArray(Direct Array Index=$Particles.PartitionParticles.Partition, Float Selection Array=DataInterface() {propertyValues {FloatData [0,0,1,1]}})
      False Float=1
     )
    )
    ScaleA=true
    ScaleRGB=true
   )
   &SkeletalMeshLocation(
    Bone / Socket Index=$Particles.PartitionParticles.Partition
    Bone Sampling Mode=::NewEnumerator4
    MaskSpawnGroup=false
    Sampled Position Offset=&FN_FrontEndVector3Modifier(
     Gameplay=&SelectVectorFromArray(
      Direct Array Index=$Particles.PartitionParticles.Partition
      Vector Selection Array=DataInterface() {propertyValues {FloatData [<-3 0 0>, <3 0 0>, <27 -3 -2>, <18 2 0>, <-2 -2 -1>, <2 2 1>]}}
     )
     Front End=&SelectVectorFromArray(
      Direct Array Index=$Particles.PartitionParticles.Partition
      Vector Selection Array=DataInterface() {propertyValues {FloatData [<-3 0 0>, <3 0 0>, <27 -3 -2>, <18 2 0>, <-2 -2 -1>, <2 2 1>]}}
     )
    )
    Skeletal Mesh=DataInterface() {propertyValues {
      DefaultMesh=null
      ExcludeBoneName=root
      OptionalComponentClass=null
      PreviewMesh=&F_MED_LemonWrath_Ochre
      SkinningMode=SkinOnTheFly
      SoftSourceActor=null
      SourceMode=Default
      UvSetIndex=0
      WholeMeshLOD=-1
      bExcludeBone=false
      bReadDeformedGeometry=true
      bRequireCurrentFrameData=true
      ComponentTags []
      FilteredBones [upperarm_l,upperarm_r,thigh_r,calf_r,lowerarm_l,lowerarm_r]
      FilteredSockets []
      MeshUserParameter {parameter {name=None, typeDefHandle {registeredTypeIndex=1}, varData []}}
      SamplingRegions []
     }}
   )
  ]
  rendererClasses [&NiagaraSpriteRendererProperties]
  renderers [{rendererClass=&NiagaraSpriteRendererProperties}]
 }
}
summary {
 systemName=NS_Character_LemonWrath_Ochre
 emitters [
  {emitterName=MainGlow, rendererClasses [&NiagaraSpriteRendererProperties]}
  {
   emitterName=IdleCubes_shoulder
   rendererClasses [&NiagaraMeshRendererProperties, &NiagaraSpriteRendererProperties]
  }
  {
   emitterName=IdleCubes_Pit001
   rendererClasses [&NiagaraMeshRendererProperties, &NiagaraSpriteRendererProperties]
  }
  {
   emitterName=IdleCubes_shoulder001
   rendererClasses [&NiagaraMeshRendererProperties, &NiagaraSpriteRendererProperties]
  }
 ]
 userVariables [
  User.bIsFrontEnd {type=NiagaraBool}
  User.bIsFrontEndPreview {type=NiagaraBool}
  User.VariantColor {type=NiagaraFloat, defaultValue=0, description=""}
 ]
}
