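# Animation Blueprint for the F_MED_LemonWrath_Ochre female body cosmetic, adding secondary motion to jacket, laces and belt.
# Copied parent pose feeds a Control Rig, then two differently tuned AnimDynamics chains join at a bool blend.
# A jacket-bone offset follows, then DefaultSlot layers onto belt and gun bones by front-end idle curve.
# Update logic drives rig alphas and backbling gating; its damped gravity and belt alpha are unused by the AnimGraph.
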AnimBlueprint /BRCosmetics/Characters/Player/Female/Medium/Bodies/F_MED_LemonWrath_Ochre/Meshes/F_MED_LemonWrath_Ochre_AnimBP
target_skeleton=&/F_MED_LemonWrath_Ochre_Skeleton
graphs [
 AnimGraph {nodes [
   n1 = Anim Dynamics - Bone: dyn_jacket_a_l_01!(ComponentPose=Local To Component(LocalPose=Use cached pose 'Base Pose'()), Alpha=1, GravityOverride:Vector="0,0,-980", ExternalForce="0,0,0") -> (Pose) {
    angularDampingOverride=1
    angularSpringConstant=8000
    bAngularSpring=true
    bOverrideAngularDamping=true
    bUseGravityOverride=true
    simulationSpace=RootRelative
    alphaBoolBlend {blendOption=Linear}
    boundBone {boneName=dyn_jacket_a_l_01}
    componentAppliedLinearAccClamp=<3000 3000 3000>
    componentLinearAccScale=<0.3 0.3 0.3>
    componentLinearVelScale=<1 1 1>
    gravityOverride {z=-980}
    physicsBodyDefinitions [{
      boundBone {boneName=dyn_jacket_a_l_01}
      boxExtents=<25 30 30>
      constraintSetup {angularLimitsMax=<0 10 15>, angularLimitsMin=<0 -30 -15>, angularTarget=<1 0 0>}
      localJointOffset=<2.5 0 0>
     }]
    relativeSpaceBone {boneName=head}
   }
   n2 = Anim Dynamics - Bone: dyn_jacket_a_r_01!(ComponentPose=@n1, Alpha=1, GravityOverride:Vector="0,0,-980", ExternalForce="0,0,0") -> (Pose) {
    angularDampingOverride=1
    angularSpringConstant=8000
    bAngularSpring=true
    bOverrideAngularDamping=true
    bUseGravityOverride=true
    simulationSpace=RootRelative
    alphaBoolBlend {blendOption=Linear}
    boundBone {boneName=dyn_jacket_a_r_01}
    componentAppliedLinearAccClamp=<3000 3000 3000>
    componentLinearAccScale=<0.3 0.3 0.3>
    componentLinearVelScale=<1 1 1>
    gravityOverride {z=-980}
    physicsBodyDefinitions [{
      boundBone {boneName=dyn_jacket_a_r_01}
      boxExtents=<25 30 30>
      constraintSetup {angularLimitsMax=<0 10 15>, angularLimitsMin=<0 -30 -15>, angularTarget=<1 0 0>}
      localJointOffset=<2.5 0 0>
     }]
    relativeSpaceBone {boneName=head}
   }
   n3 = Anim Dynamics - Bone: dyn_jacket_b_r_01!(ComponentPose=@n2, Alpha=1, GravityOverride:Vector="0,0,-980", ExternalForce="0,0,0") -> (Pose) {
    angularDampingOverride=0.9999
    angularSpringConstant=8000
    bAngularSpring=true
    bOverrideAngularDamping=true
    bUseGravityOverride=true
    simulationSpace=RootRelative
    alphaBoolBlend {blendOption=Linear}
    boundBone {boneName=dyn_jacket_b_r_01}
    componentAppliedLinearAccClamp=<3000 3000 3000>
    componentLinearAccScale=<0.25 0.25 0.25>
    componentLinearVelScale=<1 1 1>
    gravityOverride {z=-980}
    physicsBodyDefinitions [{
      boundBone {boneName=dyn_jacket_b_r_01}
      boxExtents=<20 35 35>
      constraintSetup {angularLimitsMax=<0 25 5>, angularLimitsMin=<0 -25 -5>, angularTarget=<1 0 0>}
      localJointOffset=<4.5 0 0>
     }]
    relativeSpaceBone {boneName=head}
   }
   n4 = Anim Dynamics - Bone: dyn_jacket_b_l_01!(ComponentPose=@n3, Alpha=1, GravityOverride:Vector="0,0,-980", ExternalForce="0,0,0") -> (Pose) {
    angularDampingOverride=0.9999
    angularSpringConstant=8000
    bAngularSpring=true
    bOverrideAngularDamping=true
    bUseGravityOverride=true
    simulationSpace=RootRelative
    alphaBoolBlend {blendOption=Linear}
    boundBone {boneName=dyn_jacket_b_l_01}
    componentAppliedLinearAccClamp=<3000 3000 3000>
    componentLinearAccScale=<0.25 0.25 0.25>
    componentLinearVelScale=<1 1 1>
    gravityOverride {z=-980}
    physicsBodyDefinitions [{
      boundBone {boneName=dyn_jacket_b_l_01}
      boxExtents=<20 35 35>
      constraintSetup {angularLimitsMax=<0 25 5>, angularLimitsMin=<0 -25 -5>, angularTarget=<1 0 0>}
      localJointOffset=<4.5 0 0>
     }]
    relativeSpaceBone {boneName=head}
   }
   n5 = Anim Dynamics - Bone: dyn_jacket_back_01!(
    ComponentPose=@n4
    bAlphaBoolEnabled:bool=NOT Boolean(A=Get Is Wearing Backbling())
    GravityOverride:Vector="0,0,-980"
    ExternalForce="0,0,0"
   ) -> (Pose) {
    alphaInputType=Bool
    angularDampingOverride=1
    angularSpringConstant=8000
    bAngularSpring=true
    bOverrideAngularDamping=true
    bUseGravityOverride=true
    simulationSpace=RootRelative
    alphaBoolBlend {blendOption=Linear}
    boundBone {boneName=dyn_jacket_back_01}
    componentAppliedLinearAccClamp=<3000 3000 3000>
    componentLinearAccScale=<0.3 0.3 0.3>
    componentLinearVelScale=<1 1 1>
    gravityOverride {z=-980}
    physicsBodyDefinitions [{
      boundBone {boneName=dyn_jacket_back_01}
      boxExtents=<20 30 30>
      constraintSetup {angularLimitsMax=<0 15 0>, angularLimitsMin=<0 -45 0>, angularTarget=<1 0 0>}
      localJointOffset=<7 0 0>
     }]
    relativeSpaceBone {boneName=head}
   }
   n6 = Anim Dynamics - Bone: dyn_jacket_back_01(ComponentPose=@n5, bAlphaBoolEnabled=True, GravityOverride="0,0,-980", ExternalForce="0,0,0") {
    alphaInputType=Bool
    angularDampingOverride=1
    angularSpringConstant=8000
    bAngularSpring=true
    bOverrideAngularDamping=true
    bUseGravityOverride=true
    simulationSpace=RootRelative
    alphaBoolBlend {blendOption=Linear}
    boundBone {boneName=dyn_jacket_back_01}
    componentAppliedLinearAccClamp=<3000 3000 3000>
    componentLinearAccScale=<0.3 0.3 0.3>
    componentLinearVelScale=<0.8 0.8 0.8>
    gravityOverride {z=-980}
    physicsBodyDefinitions [{
      boundBone {boneName=dyn_jacket_back_01}
      boxExtents=<20 30 30>
      constraintSetup {angularLimitsMax=<0 15 0>, angularLimitsMin=<0 -15 0>, angularTarget=<1 0 0>}
      localJointOffset=<5 0 0>
     }]
    relativeSpaceBone {boneName=head}
   }
   n7 = Anim Dynamics - Bone: dyn_laces_l_01!(ComponentPose=@n6, Alpha=1, GravityOverride:Vector="0,0,-980", ExternalForce="0,0,0") -> (Pose) {
    angularDampingOverride=0.9999
    angularSpringConstant=5000
    bAngularSpring=true
    bOverrideAngularDamping=true
    bUseGravityOverride=true
    simulationSpace=RootRelative
    alphaBoolBlend {blendOption=Linear}
    boundBone {boneName=dyn_laces_l_01}
    componentAppliedLinearAccClamp=<3000 3000 3000>
    componentLinearAccScale=<0.45 0.45 0.45>
    componentLinearVelScale=<0.45 0.45 0.45>
    gravityOverride {z=-980}
    physicsBodyDefinitions [{
      boundBone {boneName=dyn_laces_l_01}
      boxExtents=<15 20 20>
      constraintSetup {angularLimitsMax=<0 0 10>, angularLimitsMin=<0 -10 -10>, angularTarget=<1 0 0>}
      localJointOffset=<2.5 0 0>
     }]
    relativeSpaceBone {boneName=head}
   }
   n8 = Anim Dynamics - Bone: dyn_laces_l_02!(ComponentPose=@n7, Alpha=1, GravityOverride:Vector="0,0,-980", ExternalForce="0,0,0") -> (Pose) {
    angularDampingOverride=0.9999
    angularSpringConstant=5000
    bAngularSpring=true
    bOverrideAngularDamping=true
    bUseGravityOverride=true
    simulationSpace=RootRelative
    alphaBoolBlend {blendOption=Linear}
    boundBone {boneName=dyn_laces_l_02}
    componentAppliedLinearAccClamp=<3000 3000 3000>
    componentLinearAccScale=<0.45 0.45 0.45>
    componentLinearVelScale=<0.45 0.45 0.45>
    gravityOverride {z=-980}
    physicsBodyDefinitions [{
      boundBone {boneName=dyn_laces_l_02}
      boxExtents=<15 20 20>
      constraintSetup {angularLimitsMax=<0 10 10>, angularLimitsMin=<0 -10 -10>, angularTarget=<1 0 0>}
      localJointOffset=<2.5 0 0>
     }]
    relativeSpaceBone {boneName=head}
   }
   n9 = Anim Dynamics - Bone: dyn_laces_r_01!(ComponentPose=@n8, Alpha=1, GravityOverride:Vector="0,0,-980", ExternalForce="0,0,0") -> (Pose) {
    angularDampingOverride=0.9999
    angularSpringConstant=5000
    bAngularSpring=true
    bOverrideAngularDamping=true
    bUseGravityOverride=true
    simulationSpace=RootRelative
    alphaBoolBlend {blendOption=Linear}
    boundBone {boneName=dyn_laces_r_01}
    componentAppliedLinearAccClamp=<3000 3000 3000>
    componentLinearAccScale=<0.45 0.45 0.45>
    componentLinearVelScale=<0.45 0.45 0.45>
    gravityOverride {z=-980}
    physicsBodyDefinitions [{
      boundBone {boneName=dyn_laces_r_01}
      boxExtents=<15 20 20>
      constraintSetup {angularLimitsMax=<0 0 10>, angularLimitsMin=<0 -10 -10>, angularTarget=<1 0 0>}
      localJointOffset=<2.5 0 0>
     }]
    relativeSpaceBone {boneName=head}
   }
   n10 = Anim Dynamics - Bone: dyn_laces_r_02!(ComponentPose=@n9, Alpha=1, GravityOverride:Vector="0,0,-980", ExternalForce="0,0,0") -> (Pose) {
    angularDampingOverride=0.9999
    angularSpringConstant=5000
    bAngularSpring=true
    bOverrideAngularDamping=true
    bUseGravityOverride=true
    simulationSpace=RootRelative
    alphaBoolBlend {blendOption=Linear}
    boundBone {boneName=dyn_laces_r_02}
    componentAppliedLinearAccClamp=<3000 3000 3000>
    componentLinearAccScale=<0.45 0.45 0.45>
    componentLinearVelScale=<0.45 0.45 0.45>
    gravityOverride {z=-980}
    physicsBodyDefinitions [{
      boundBone {boneName=dyn_laces_r_02}
      boxExtents=<15 20 20>
      constraintSetup {angularLimitsMax=<0 10 10>, angularLimitsMin=<0 -10 -10>, angularTarget=<1 0 0>}
      localJointOffset=<2.5 0 0>
     }]
    relativeSpaceBone {boneName=head}
   }
   n11 = Anim Dynamics - Bone: dyn_belt_01!(ComponentPose=@n10, Alpha=1, GravityOverride:Vector="0,0,-980", ExternalForce="0,0,0") -> (Pose) {
    angularDampingOverride=0.9999
    angularSpringConstant=6000
    bOverrideAngularDamping=true
    bUseGravityOverride=true
    simulationSpace=RootRelative
    alphaBoolBlend {blendOption=Linear}
    boundBone {boneName=dyn_belt_01}
    componentAppliedLinearAccClamp=<3000 3000 3000>
    componentLinearAccScale=<0.45 0.45 0.45>
    componentLinearVelScale=<0.45 0.45 0.45>
    gravityOverride {z=-980}
    physicsBodyDefinitions [{
      boundBone {boneName=dyn_belt_01}
      boxExtents=<30 20 20>
      constraintSetup {angularLimitsMax=<0 0 5>, angularLimitsMin=<0 -10 -5>, angularTarget=<1 0 0>}
      localJointOffset=<2.5 0 -2.5>
     }]
    relativeSpaceBone {boneName=head}
   }
   n12 = Component To Local(ComponentPose=@n11)
   n13 = Anim Dynamics - Bone: dyn_jacket_a_l_01(ComponentPose=Local To Component(LocalPose=Use cached pose 'Base Pose'()), Alpha=1, GravityOverride="0,0,-980", ExternalForce="0,0,0") {
    angularDampingOverride=1
    angularSpringConstant=8000
    bAngularSpring=true
    bOverrideAngularDamping=true
    bUseGravityOverride=true
    simulationSpace=RootRelative
    alphaBoolBlend {blendOption=Linear}
    boundBone {boneName=dyn_jacket_a_l_01}
    gravityOverride {z=-980}
    physicsBodyDefinitions [{
      boundBone {boneName=dyn_jacket_a_l_01}
      boxExtents=<25 30 30>
      constraintSetup {angularLimitsMax=<0 10 15>, angularLimitsMin=<0 -30 -15>, angularTarget=<1 0 0>}
      localJointOffset=<2.5 0 0>
     }]
    relativeSpaceBone {boneName=head}
    simSpaceSettings {maxAngularAcceleration=30, maxAngularVelocity=6, simSpaceAngularAlpha=1}
   }
   n14 = Anim Dynamics - Bone: dyn_jacket_a_r_01(ComponentPose=@n13, Alpha=1, GravityOverride="0,0,-980", ExternalForce="0,0,0") {
    angularDampingOverride=1
    angularSpringConstant=8000
    bAngularSpring=true
    bOverrideAngularDamping=true
    bUseGravityOverride=true
    simulationSpace=RootRelative
    alphaBoolBlend {blendOption=Linear}
    boundBone {boneName=dyn_jacket_a_r_01}
    gravityOverride {z=-980}
    physicsBodyDefinitions [{
      boundBone {boneName=dyn_jacket_a_r_01}
      boxExtents=<25 30 30>
      constraintSetup {angularLimitsMax=<0 10 15>, angularLimitsMin=<0 -30 -15>, angularTarget=<1 0 0>}
      localJointOffset=<2.5 0 0>
     }]
    relativeSpaceBone {boneName=head}
    simSpaceSettings {maxAngularAcceleration=30, maxAngularVelocity=6, simSpaceAngularAlpha=1}
   }
   n15 = Anim Dynamics - Bone: dyn_jacket_b_r_01(ComponentPose=@n14, Alpha=1, GravityOverride="0,0,-980", ExternalForce="0,0,0") {
    angularDampingOverride=0.9999
    angularSpringConstant=8000
    bAngularSpring=true
    bOverrideAngularDamping=true
    bUseGravityOverride=true
    simulationSpace=RootRelative
    alphaBoolBlend {blendOption=Linear}
    boundBone {boneName=dyn_jacket_b_r_01}
    gravityOverride {z=-980}
    physicsBodyDefinitions [{
      boundBone {boneName=dyn_jacket_b_r_01}
      boxExtents=<20 35 35>
      constraintSetup {angularLimitsMax=<0 25 5>, angularLimitsMin=<0 -25 -5>, angularTarget=<1 0 0>}
      localJointOffset=<4.5 0 0>
     }]
    relativeSpaceBone {boneName=head}
    simSpaceSettings {maxAngularAcceleration=30, maxAngularVelocity=6, simSpaceAngularAlpha=1}
   }
   n16 = Anim Dynamics - Bone: dyn_jacket_b_l_01(ComponentPose=@n15, Alpha=1, GravityOverride="0,0,-980", ExternalForce="0,0,0") {
    angularDampingOverride=0.9999
    angularSpringConstant=8000
    bAngularSpring=true
    bOverrideAngularDamping=true
    bUseGravityOverride=true
    simulationSpace=RootRelative
    alphaBoolBlend {blendOption=Linear}
    boundBone {boneName=dyn_jacket_b_l_01}
    gravityOverride {z=-980}
    physicsBodyDefinitions [{
      boundBone {boneName=dyn_jacket_b_l_01}
      boxExtents=<20 35 35>
      constraintSetup {angularLimitsMax=<0 25 5>, angularLimitsMin=<0 -25 -5>, angularTarget=<1 0 0>}
      localJointOffset=<4.5 0 0>
     }]
    relativeSpaceBone {boneName=head}
    simSpaceSettings {maxAngularAcceleration=30, maxAngularVelocity=6, simSpaceAngularAlpha=1}
   }
   n17 = Anim Dynamics - Bone: dyn_jacket_back_01(ComponentPose=@n16, bAlphaBoolEnabled=NOT Boolean(A=Get Is Wearing Backbling()), GravityOverride="0,0,-980", ExternalForce="0,0,0") {
    alphaInputType=Bool
    angularDampingOverride=1
    angularSpringConstant=8000
    bAngularSpring=true
    bOverrideAngularDamping=true
    bUseGravityOverride=true
    simulationSpace=RootRelative
    alphaBoolBlend {blendOption=Linear}
    boundBone {boneName=dyn_jacket_back_01}
    gravityOverride {z=-980}
    physicsBodyDefinitions [{
      boundBone {boneName=dyn_jacket_back_01}
      boxExtents=<20 30 30>
      constraintSetup {angularLimitsMax=<0 15 0>, angularLimitsMin=<0 -45 0>, angularTarget=<1 0 0>}
      localJointOffset=<7 0 0>
     }]
    relativeSpaceBone {boneName=head}
    simSpaceSettings {maxAngularAcceleration=30, maxAngularVelocity=6, simSpaceAngularAlpha=1}
   }
   n18 = Anim Dynamics - Bone: dyn_jacket_back_01(ComponentPose=@n17, bAlphaBoolEnabled=True, GravityOverride="0,0,-980", ExternalForce="0,0,0") {
    alphaInputType=Bool
    angularDampingOverride=1
    angularSpringConstant=8000
    bAngularSpring=true
    bOverrideAngularDamping=true
    bUseGravityOverride=true
    simulationSpace=RootRelative
    alphaBoolBlend {blendOption=Linear}
    boundBone {boneName=dyn_jacket_back_01}
    gravityOverride {z=-980}
    physicsBodyDefinitions [{
      boundBone {boneName=dyn_jacket_back_01}
      boxExtents=<20 30 30>
      constraintSetup {angularLimitsMax=<0 15 0>, angularLimitsMin=<0 -15 0>, angularTarget=<1 0 0>}
      localJointOffset=<5 0 0>
     }]
    relativeSpaceBone {boneName=head}
    simSpaceSettings {maxAngularAcceleration=30, maxAngularVelocity=6, simSpaceAngularAlpha=1}
   }
   n19 = Anim Dynamics - Bone: dyn_laces_l_01(ComponentPose=@n18, Alpha=1, GravityOverride="0,0,-980", ExternalForce="0,0,0") {
    angularDampingOverride=0.9995
    angularSpringConstant=5000
    bAngularSpring=true
    bOverrideAngularDamping=true
    bUseGravityOverride=true
    simulationSpace=RootRelative
    alphaBoolBlend {blendOption=Linear}
    boundBone {boneName=dyn_laces_l_01}
    gravityOverride {z=-980}
    physicsBodyDefinitions [{
      sphereCollisionRadius=5
      boundBone {boneName=dyn_laces_l_01}
      boxExtents=<15 20 20>
      constraintSetup {angularLimitsMax=<0 0 10>, angularLimitsMin=<0 -20 -10>, angularTarget=<1 0 0>}
      localJointOffset=<2.5 0 0>
     }]
    relativeSpaceBone {boneName=head}
    simSpaceSettings {maxAngularAcceleration=30, maxAngularVelocity=6, simSpaceAngularAlpha=1}
   }
   n20 = Anim Dynamics - Bone: dyn_laces_l_02(ComponentPose=@n19, Alpha=1, GravityOverride="0,0,-980", ExternalForce="0,0,0") {
    angularDampingOverride=0.9995
    angularSpringConstant=5000
    bAngularSpring=true
    bOverrideAngularDamping=true
    bUseGravityOverride=true
    simulationSpace=RootRelative
    alphaBoolBlend {blendOption=Linear}
    boundBone {boneName=dyn_laces_l_02}
    gravityOverride {z=-980}
    physicsBodyDefinitions [{
      boundBone {boneName=dyn_laces_l_02}
      boxExtents=<15 20 20>
      constraintSetup {angularLimitsMax=<0 10 10>, angularLimitsMin=<0 -10 -10>, angularTarget=<1 0 0>}
      localJointOffset=<2.5 0 0>
     }]
    relativeSpaceBone {boneName=head}
    simSpaceSettings {maxAngularAcceleration=30, maxAngularVelocity=6, simSpaceAngularAlpha=1}
   }
   n21 = Anim Dynamics - Bone: dyn_laces_r_01(ComponentPose=@n20, Alpha=1, GravityOverride="0,0,-980", ExternalForce="0,0,0") {
    angularDampingOverride=0.9999
    angularSpringConstant=5000
    bAngularSpring=true
    bOverrideAngularDamping=true
    bUseGravityOverride=true
    simulationSpace=RootRelative
    alphaBoolBlend {blendOption=Linear}
    boundBone {boneName=dyn_laces_r_01}
    gravityOverride {z=-980}
    physicsBodyDefinitions [{
      sphereCollisionRadius=5
      boundBone {boneName=dyn_laces_r_01}
      boxExtents=<16 20 20>
      constraintSetup {angularLimitsMax=<0 0 10>, angularLimitsMin=<0 -20 -10>, angularTarget=<1 0 0>}
      localJointOffset=<2.5 0 0>
     }]
    relativeSpaceBone {boneName=head}
    simSpaceSettings {maxAngularAcceleration=25, maxAngularVelocity=5, simSpaceAngularAlpha=1}
   }
   n22 = Anim Dynamics - Bone: dyn_laces_r_02(ComponentPose=@n21, Alpha=1, GravityOverride="0,0,-980", ExternalForce="0,0,0") {
    angularDampingOverride=0.9999
    angularSpringConstant=5000
    bAngularSpring=true
    bOverrideAngularDamping=true
    bUseGravityOverride=true
    simulationSpace=RootRelative
    alphaBoolBlend {blendOption=Linear}
    boundBone {boneName=dyn_laces_r_02}
    gravityOverride {z=-980}
    physicsBodyDefinitions [{
      boundBone {boneName=dyn_laces_r_02}
      boxExtents=<16 20 20>
      constraintSetup {angularLimitsMax=<0 10 10>, angularLimitsMin=<0 -10 -10>, angularTarget=<1 0 0>}
      localJointOffset=<2.5 0 0>
     }]
    relativeSpaceBone {boneName=head}
    simSpaceSettings {maxAngularAcceleration=25, maxAngularVelocity=5, simSpaceAngularAlpha=1}
   }
   n23 = Anim Dynamics - Bone: dyn_belt_01(ComponentPose=@n22, Alpha=1, GravityOverride="0,0,-980", ExternalForce="0,0,0") {
    angularDampingOverride=0.9999
    angularSpringConstant=6000
    bOverrideAngularDamping=true
    bUseGravityOverride=true
    simulationSpace=RootRelative
    alphaBoolBlend {blendOption=Linear}
    boundBone {boneName=dyn_belt_01}
    gravityOverride {z=-980}
    physicsBodyDefinitions [{
      boundBone {boneName=dyn_belt_01}
      boxExtents=<30 20 20>
      constraintSetup {angularLimitsMax=<0 0 5>, angularLimitsMin=<0 -10 -5>, angularTarget=<1 0 0>}
      localJointOffset=<2.5 0 -2.5>
     }]
    relativeSpaceBone {boneName=head}
    simSpaceSettings {maxAngularAcceleration=30, maxAngularVelocity=6, simSpaceAngularAlpha=1}
   }
   n24 = Component To Local(ComponentPose=@n23)
   n25 = "Transform (Modify) Bone - Bone: dyn_jacket_b_l_base"(
    ComponentPose=Local To Component(LocalPose=Blend Poses by bool(BlendPose_0=@n24, BlendPose_1=@n12))
    Alpha=1
    Translation="0,0,-1"
    Rotation="-12.5,0,0"
    Scale="1,1,1"
   ) {
    rotationMode=BMM_Additive
    rotationSpace=BCS_BoneSpace
    translationMode=BMM_Additive
    translationSpace=BCS_BoneSpace
    boneToModify {boneName=dyn_jacket_b_l_base}
    rotation {pitch=-12.5}
    translation {z=-1}
   }
   n26 = Layered blend per bone(
    BasePose=Component To Local(ComponentPose=@n25)
    BlendPoses_0=Slot 'DefaultSlot'()
    BlendWeights_0=Get customIdleIsPlayingFloat() -> (customIdleIsPlayingFloat:float)
   ) {layerSetup [{branchFilters [{boneName=dyn_belt_base}, {boneName=dyn_belt_01}, {boneName=dyn_gun_base}, {boneName=dyn_gun_01}]}]}
   Output Pose(Result=@n26) {layerGroup=None}
   n27 = Control Rig!(
    Source=Copy Pose From Mesh() {bUseAttachedParent=true}
    Alpha
    GunPoseModifyAlpha_Emote:double=Get GunPoseModifyAlpha_Emote() -> (GunPoseModifyAlpha_Emote:float)
    JacketPoseModifyAlpha_CrouchMove:double=Get JacketPoseModifyAlpha_CrouchMove() -> (JacketPoseModifyAlpha_CrouchMove:float)
    inFE:bool=Get bIsInFrontEnd()
   ) -> (Pose) {controlRigAssetReference {blueprintRigClass=&/F_MED_LemonWrath_Ochre_CtrlRig.F_MED_LemonWrath_Ochre_CtrlRig_C}}
   Save cached pose 'Base Pose'(Pose=@n27) {cacheName=Base Pose}
  ]}
 EventGraph {nodes [
   n1 = Event BlueprintUpdateAnimation()
   > n2 = Sequence()
   > Set Delta Time X(Delta Time X=@n1.DeltaTimeX)
   n3 = Get bIsSkydiveFloating()
   > Set GravityOverride(GravityOverride=DampVector!(
     Value:"Vector (by ref)"=Get GravityOverride()
     Target:"Vector (by ref)"=SelectVector(A="0,0,350", B="0,0,-980", bPickA=@n3)
     DeltaTime:float=Get Delta Time X()
     SmoothingTime:float=0.1
    ) -> (ReturnValue:Vector))
   n4 = Get bIsMoving2D()
   n5 = Get bIsPlayingEmote()
   n6 = OR Boolean(
    A=Get bIsCrouching()
    B=AND Boolean!(A=NOT Boolean(A=@n4), B=Get bIsOnGround() -> (bIsOnGround:bool), C:bool=NOT Boolean(A=@n5)) -> (ReturnValue)
   )
   > Set BeltDynamicsAlpha(BeltDynamicsAlpha:float=DampFloat!(
     Value:float=Get BeltDynamicsAlpha() -> (BeltDynamicsAlpha:float)
     Target:float=SelectFloat(A=0, B=1, bPickA=@n6)
     DeltaTime:float=Get Delta Time X()
     SmoothingTime:float=0.05
    ) -> (ReturnValue:float)) -> (Output_Get:float)
   > Set Is Wearing Backbling(Is Wearing Backbling=CheckForBackbling())
   n7 = Get bIsPlayingEmote()
   > Set GunPoseModifyAlpha_Emote(GunPoseModifyAlpha_Emote:float=DampFloat(Value=Get GunPoseModifyAlpha_Emote(), Target=SelectFloat(A=1, B=0, bPickA=@n7), DeltaTime=Get Delta Time X(), SmoothingTime=0.1)) -> (Output_Get:float)
   n8 = AND Boolean!(A=Get bIsCrouching(), B=Get bIsMoving2D()) -> (ReturnValue)
   > Set JacketPoseModifyAlpha_CrouchMove(JacketPoseModifyAlpha_CrouchMove:float=DampFloat(Value=Get JacketPoseModifyAlpha_CrouchMove(), Target=SelectFloat(A=1, B=0, bPickA=@n8), DeltaTime=Get Delta Time X(), SmoothingTime=0.1)) -> (Output_Get:float)
   Branch(execute=@n2.then_1, Condition=Get bIsInFrontEnd())
   > Set customIdleIsPlayingFloat(customIdleIsPlayingFloat:float=GetCurveValue(self=GetOwnerFrontEndAnimInstance(), CurveName=Frontend_Idle_Bypass)) -> (Output_Get:float)
  ]}
]
variables [
 BeltDynamicsAlpha {type=number}
 customIdleIsPlayingFloat {type=number}
 Delta Time X {type=number}
 GravityOverride {type=Vector}
 GunPoseModifyAlpha_Emote {type=number}
 Is Wearing Backbling {type=boolean}
 JacketPoseModifyAlpha_CrouchMove {type=number}
]
